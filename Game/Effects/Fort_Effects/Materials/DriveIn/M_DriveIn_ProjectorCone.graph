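# Unlit, two-sided translucent material for a drive-in projector's light cone, usable on mesh particles.
# Three panned, rotated samples of a gradient-section texture drive shifting brightness and hue on an orange base, partly desaturated.
# Particle color and time-of-day tints scale that color before emissive adjustment.
# A soft-edge mask multiplies Fresnel, view-angle fade and a vertical gradient peaking mid-length, feeding synced-distance and near-camera fade nodes.
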Material /Game/Effects/Fort_Effects/Materials/DriveIn/M_DriveIn_ProjectorCone
graph {
 properties {
  bEnableResponsiveAA=true
  bUsedWithMeshParticles=true
  blendMode=BLEND_Translucent
  pixelDepthOffsetMode=PDOM_Legacy
  previewMesh=/Game/Effects/Fort_Effects/Mesh/Rays/S_Billboard_LightCone
  shadingModel=MSM_Unlit
  twoSided=true
  usedShadingModels=Unlit
 }
 nodes [
  n1 = Constant() {r=1.5}
  MaterialFunctionCall(Axis=Constant3Vector() {constant=<rgba 0 0 1 0>}, Exponent=@n1) {materialFunction=&AxisAlignedFresnel}
  n2 = Constant() {r=4}
  n3 = MaterialFunctionCall() {materialFunction=&LinearGradient}
  n4 = Multiply(A=@n3.VGradient, B=OneMinus(None=@n3.VGradient))
  n5 = Multiply(
   A=Multiply(
    A=Fresnel(ExponentIn=@n1)
    B=MaterialFunctionCall("FadeSensitivity (optional)"=@n2) {materialFunction=&FadeBasedOnViewAngle}.Result
   )
   B=@n4
  )
  MaterialFunctionCall(
   Opacity In=@n5
   Fade_Start_Distance=ScalarParameter() {defaultValue=1024, parameterName="Fade_Start_Distance (S)"}
   Fade_End_Distance=ScalarParameter() {defaultValue=512, parameterName="Fade_End_Distance (S)"}
  ) {materialFunction=&MF_Distance_SyncedFade_FX}
  n6 = Panner(Coordinate=TextureCoordinate(), Time=Time()) {speedX=0.12, speedY=0.25}
  n7 = ScalarParameter() {defaultValue=0.25, parameterName="Rotation Angle (0-1) (S)"}
  n8 = ScalarParameter() {defaultValue=0.15, parameterName=A}
  n9 = ScalarParameter() {defaultValue=1, parameterName=B}
  n10 = Panner(Coordinate=TextureCoordinate(), Time=Time()) {speedX=0.2, speedY=0.3}
  n11 = Panner(Coordinate=TextureCoordinate(), Time=Time()) {speedX=-0.1, speedY=-0.2}
  n12 = Multiply(
   A=TextureSample(UVs=MaterialFunctionCall(UVs=@n10, "Rotation Angle (0-1)"=@n7) {materialFunction=&CustomRotator}.Rotated Values) {texture=/Game/Global/Textures/Default/Masks_and_Gradients/T_Vert_Gradient_Sections}.G
   B=TextureSample(UVs=MaterialFunctionCall(UVs=@n11, "Rotation Angle (0-1)"=@n7) {materialFunction=&CustomRotator}.Rotated Values) {texture=&T_Vert_Gradient_Sections}.G
  )
  n13 = MaterialFunctionCall(
   Hue Shift Percentage=TextureSample(UVs=MaterialFunctionCall(UVs=@n6, "Rotation Angle (0-1)"=@n7) {materialFunction=&CustomRotator}.Rotated Values) {texture=&T_Vert_Gradient_Sections}.G
   Texture=Multiply(A=LinearInterpolate(A=@n8, B=@n9, Alpha=@n12), B=Constant3Vector() {constant=<rgba 1 0.25 0.1 0>})
  ) {materialFunction=&HueShift}
  n14 = ScalarParameter() {defaultValue=0.5, parameterName=DesatFraction}
  n15 = ParticleColor()
  n16 = MaterialFunctionCall(
   Color_Morning=VectorParameter() {parameterName="Color_Morning (V3)", defaultValue=<rgba 1 1 1 1>}.RGB
   Color_Day=VectorParameter() {parameterName="Color_Day (V3)", defaultValue=<rgba 0.75 0.75 0.75 1>}.RGB
   Color_Evening=VectorParameter() {parameterName="Color_Evening (V3)", defaultValue=<rgba 1.1 1.1 1.1 1.1>}.RGB
   Color_Night=VectorParameter() {parameterName="Color_Night (V3)", defaultValue=<rgba 1.25 1.25 1.25 1.25>}.RGB
  ) {materialFunction=&Color_Time_of_Day}
  MaterialFunctionCall(Emissive=Multiply(A=Desaturation(None=@n13.Result, Fraction=@n14), B=Multiply(A=@n15.RGB, B=@n16.Result))) {materialFunction=&MF_FX_EmissiveAdjustments}
  Multiply(
   A=MaterialFunctionCall(Distance=ScalarParameter() {defaultValue=2048, parameterName="Distance (S)"}) {materialFunction=&Near_Camera_Fade}.Result
   B=@n5
  )
  Multiply(A=@n4, B=@n4)
 ]
}
